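# Looping ambient sound cue for a small city burn barrel fire, in the ambient sound class with a small burn barrel attenuation asset.
# A random node picks one of three looping fire urn waves, and a modulator randomizes its pitch.
# Two nested, endlessly looping envelopers of different lengths add slow, slight dips in pitch and volume to vary the loop.
# Its own concurrency override caps simultaneous instances at a low count.
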SoundCue /Game/Sounds/Asteria/Environment_Actors/BurnBarrel/Spot_City_BurnBarrel_Small_Loop
root=Enveloper_3
nodes [
 {
  id=Enveloper_0
  children [Modulator_0]
  properties {
   bLoop=true
   bLoopIndefinitely=true
   loopEnd=30
   pitchCurve {editorCurveData {keys [
      {arriveTangent=-0.0013, interpMode=RCIM_Cubic, leaveTangent=-0.0013, time=0}
      <at 15 0.98 RCIM_Cubic>
      {arriveTangent=0.0013, interpMode=RCIM_Cubic, leaveTangent=0.0013, time=30}
     ]}}
   volumeCurve {editorCurveData {keys [<at 0 1>]}}
  }
 }
 {
  id=Enveloper_3
  children [Enveloper_0]
  properties {
   bLoop=true
   bLoopIndefinitely=true
   loopEnd=20
   pitchCurve {editorCurveData {keys [{time=0}]}}
   volumeCurve {editorCurveData {keys [
      {arriveTangent=-0.01, interpMode=RCIM_Cubic, leaveTangent=-0.01, time=0, value=1}
      <at 10 0.9 RCIM_Cubic>
      {arriveTangent=0.01, interpMode=RCIM_Cubic, leaveTangent=0.01, time=20, value=1}
     ]}}
  }
 }
 {id=Modulator_0, children [Random_1], properties {pitchMax=1.1, pitchMin=0.8, volumeMax=1}}
 {id=Random_1, children [WavePlayer_0,WavePlayer_4,WavePlayer_5], properties {bShouldExcludeFromBranchCulling=true, weights [1,1,1]}}
 {id=WavePlayer_0, properties {bLooping=true, soundWaveAssetPtr=/Game/Sounds/Trans_AmbientLoops/Fire/Fire_Urn_01}}
 {id=WavePlayer_4, properties {bLooping=true, soundWaveAssetPtr=/Game/Sounds/Trans_AmbientLoops/Fire/Fire_Urn_02}}
 {id=WavePlayer_5, properties {bLooping=true, soundWaveAssetPtr=/Game/Sounds/Trans_AmbientLoops/Fire/Fire_Urn_03}}
]
properties {
 attenuationSettings=&/ATT_BurnBarrel_Small_Loop
 bOverrideConcurrency=true
 maxDistance=2750
 soundClassObject=&Fort_Ambient_Athena
 subtitlePriority=5000
 volumeMultiplier=0.5
 attenuationOverrides {
  bEnableFocusInterpolation=true
  bEnableListenerFocus=true
  bEnableOcclusion=true
  dBAttenuationAtMax=-40
  falloffDistance=512
  focusAzimuth=15
  focusVolumeAttenuation=1.5
  nonFocusAzimuth=45
  nonFocusVolumeAttenuation=0.5
  occlusionInterpolationTime=1
  occlusionLowPassFilterFrequency=10000
  occlusionTraceChannel=ECC_GameTraceChannel11
  occlusionVolumeAttenuation=0.5
  stereoSpread=512
  attenuationShapeExtents=<256 64.5942 90>
 }
 concurrencyOverrides {maxCount=2, voiceStealReleaseTime=0.5, volumeScaleAttackTime=0.5}
}
